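# Material function for the JollyTroll backpack that composites animated flipbook eyes onto incoming material attributes.
# Selectable UVs are rotated, offset, optionally bump-offset and center-scaled, then sampled through CustomNumbers_SubUV.
# Blink and squint pose parameters are thresholded into frame values, with the maximum picking the frame unless a debug switch overrides it.
# Flipbook color and alpha feed blends that override base color, opacity mask, roughness, metallic, optional specular and emissive.
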MaterialFunction /Game/Accessories/FORT_Backpacks/Backpack_M_MED_JollyTroll/Materials/MF_JollyTroll_Flipbook
graph {nodes [
  n1 = GetMaterialAttributes(None=FunctionInput() {inputName=Material, inputType=FunctionInput_MaterialAttributes}) {attributeGetTypes [
    #1
    #2
    #3
    #4
    #5
    #6
   ]}
  n2 = SetMaterialAttributes!(
   MaterialAttributes=@n1.MaterialAttributes
   BaseColor=NamedRerouteUsage()
   OpacityMask=NamedRerouteUsage()
   Roughness=NamedRerouteUsage()
   Metallic=NamedRerouteUsage()
   Specular=NamedRerouteUsage()
   EmissiveColor=NamedRerouteUsage()
  ) {attributeSetTypes [
    #1
    #2
    #3
    #4
    #5
    #6
   ]}
  FunctionOutput(None=@n2)
  NamedRerouteDeclaration(None=@n1.BaseColor) {name=inBaseColor, nodeColor=<rgba 1 0 0.7266 1>}
  NamedRerouteDeclaration(None=LinearInterpolate(
    A=TextureCoordinate()
    B=TextureCoordinate() {coordinateIndex=1}
    Alpha=ScalarParameter() {group=JollyTroll Flipbook, parameterName=use Second UVChannel}
   )) {name=Master SecondUVChannel, nodeColor=<rgba 0 0.1562 1 1>}
  n3 = NamedRerouteUsage()
  n4 = NamedRerouteUsage()
  n5 = ScalarParameter() {defaultValue=1, group=JollyTroll Flipbook, parameterName=use AlphaAsMask}
  NamedRerouteDeclaration(None=Saturate(None=LinearInterpolate(A=@n3, B=@n4, Alpha=@n5))) {name=outOpacity, nodeColor=<rgba 0.6172 0 1 1>}
  NamedRerouteDeclaration(None=@n1.Specular) {name=inSpec, nodeColor=<rgba 1 0.9141 0 1>}
  NamedRerouteDeclaration(None=LinearInterpolate(
    A=NamedRerouteUsage()
    B=ScalarParameter() {defaultValue=1, group=JollyTroll Flipbook, parameterName=Flipbook Roughness}
    Alpha=NamedRerouteUsage()
   )) {name=outRoughness, nodeColor=<rgba 0 1 0.8359 1>}
  NamedRerouteDeclaration(None=LinearInterpolate(A=NamedRerouteUsage(), B=ScalarParameter() {group=JollyTroll Flipbook, parameterName=Flipbook Metallic}, Alpha=NamedRerouteUsage())) {name=outMetallic, nodeColor=<rgba 0 0.0625 1 1>}
  NamedRerouteDeclaration(None=ScalarParameter() {defaultValue=1, group=JollyTroll Flipbook, parameterName=Affects BaseColor}) {name=affectsBaseColor, nodeColor=<rgba 1 0.9844 0 1>}
  n6 = NamedRerouteUsage()
  n7 = ScalarParameter() {group=JollyTroll Flipbook, parameterName=Flipbook Specular}
  n8 = NamedRerouteUsage()
  NamedRerouteDeclaration(None=StaticSwitchParameter(True=LinearInterpolate(A=@n6, B=@n7, Alpha=@n8), False=@n6) {group=JollyTroll Flipbook, parameterName=useFlipbookSpec}) {name=outSpec, nodeColor=<rgba 1 0 0.5391 1>}
  NamedRerouteDeclaration(None=@n1.EmissiveColor) {name=inEmissive, nodeColor=<rgba 1 0.8203 0 1>}
  NamedRerouteDeclaration(None=LinearInterpolate(A=NamedRerouteUsage(), B=Constant(), Alpha=NamedRerouteUsage())) {name=outEmissive, nodeColor=<rgba 0.8281 0 1 1>}
  n9 = ScalarParameter() {defaultValue=0.5, group=JollyTroll Flipbook, parameterName=Flipbook scale Eyes}
  n10 = ScalarParameter() {defaultValue=1, group=JollyTroll Flipbook, parameterName=Flipbook scale V relative Eyes}
  n11 = MaterialFunctionCall(UVs=NamedRerouteUsage(), Texture Scale=AppendVector(A=@n9, B=Multiply(A=@n9, B=@n10))) {materialFunction=&ScaleUVsByCenter}
  NamedRerouteDeclaration(None=@n11.0-1 mask) {name=Eye 01Mask, nodeColor=<rgba 0 0.1328 1 1>}
  n12 = ScalarParameter() {defaultValue=2, group=JollyTroll Flipbook, parameterName=Face Intensity}
  n13 = ScalarParameter() {defaultValue=8, group=JollyTroll Flipbook, parameterName=SubImages X Eyes}
  n14 = ScalarParameter() {defaultValue=8, group=JollyTroll Flipbook, parameterName=SubImages Y Eyes}
  n15 = ScalarParameter() {group=JollyTroll Flipbook, parameterName=Frame Value Eye Default}
  n16 = If(
   A=ScalarParameter() {group=Banana, parameterName=R_blink_pose}
   B=ScalarParameter() {defaultValue=0.1, group=Banana, parameterName=Threshold R Blink}
   "A > B"=ScalarParameter() {defaultValue=1, group=JollyTroll Flipbook, parameterName=Frame Value Blink}
   "A == B"=@n15
   "A < B"=@n15
  )
  n17 = If(
   A=ScalarParameter() {group=Banana, parameterName=R_squint_inner_pose}
   B=ScalarParameter() {defaultValue=0.1, group=Banana, parameterName=Threshold Squint}
   "A > B"=ScalarParameter() {defaultValue=2, group=JollyTroll Flipbook, parameterName=Frame Value Squint}
   "A == B"=@n15
   "A < B"=@n15
  )
  n18 = MaterialFunctionCall!(
   Texture=NamedRerouteUsage()
   UVs=@n11.UVs
   SubImages=LinearInterpolate(
    A=ScalarParameter() {defaultValue=8, group=JollyTroll Flipbook, parameterName=SubImages Eyes}
    B=AppendVector(A=@n13, B=@n14)
    Alpha=ScalarParameter() {group=JollyTroll Flipbook, parameterName=useNonSquareTexture Eyes}
   )
   Frame=StaticSwitchParameter(True=ScalarParameter() {desc=future eye values, group=JollyTroll Flipbook, parameterName=Eye Frames Test}, False=Max(A=@n16, B=@n17)) {group=JollyTroll Flipbook, parameterName=useFlipbookDebug}
  ) {materialFunction=/Game/Weapons/WeaponSkins/MaterialFunction/CustomNumbers_SubUV}
  n19 = NamedRerouteDeclaration(None=@n18.Alpha) {name=Eye Alpha, nodeColor=<rgba 0 0.6484 1 1>}
  n20 = Multiply(A=Subtract(A=Ceil(None=@n19), B=OneMinus(None=@n19)), B=NamedRerouteUsage())
  NamedRerouteDeclaration(None=Saturate(None=Multiply(A=@n12, B=@n20))) {name=faceAlpha, nodeColor=<rgba 0.6406 0 1 1>}
  n21 = TextureObjectParameter() {
   group=JollyTroll Flipbook
   parameterName=Flipbook Texture
   texture=/Game/Accessories/FORT_Backpacks/Backpack_M_MED_JollyTroll/Textures/T_JollyTroll_Backpack_Flipbook
  }
  NamedRerouteDeclaration(None=@n21) {name=TextureFlipbook, nodeColor=<rgba 0.6641 1 0 1>}
  n22 = Add(
   A=Rotator(Coordinate=NamedRerouteUsage(), Time=ScalarParameter() {defaultValue=10, group=JollyTroll Flipbook, parameterName=Rotation})
   B=AppendVector(
    A=ScalarParameter() {defaultValue=0.125, group=JollyTroll Flipbook, parameterName=Flipbook X Position}
    B=ScalarParameter() {group=JollyTroll Flipbook, parameterName=Flipbook Y Position}
   )
  )
  n23 = ScalarParameter() {defaultValue=-0.25, group=JollyTroll Flipbook, parameterName=Bumpoffset Height}
  NamedRerouteDeclaration(None=LinearInterpolate(
    A=@n22
    B=BumpOffset(Coordinate=@n22, Height=@n23)
    Alpha=ScalarParameter() {defaultValue=1, group=JollyTroll Flipbook, parameterName=useBumpOffsetEyes}
   )) {name=bumpOffset, nodeColor=<rgba 1 0 0.1172 1>}
  NamedRerouteDeclaration(None=@n1.OpacityMask) {name=inOpacity, nodeColor=<rgba 1 0 0.9141 1>}
  NamedRerouteDeclaration(None=@n18.RGB) {name=Eye RGB, nodeColor=<rgba 0.7109 1 0 1>}
  NamedRerouteDeclaration(None=@n1.Roughness) {name=inRoughness, nodeColor=<rgba 0 1 0.25 1>}
  NamedRerouteDeclaration(None=@n1.Metallic) {name=inMetallic, nodeColor=<rgba 0.3359 1 0 1>}
  NamedRerouteDeclaration(None=LinearInterpolate(A=NamedRerouteUsage(), B=NamedRerouteUsage(), Alpha=NamedRerouteUsage()) {constB=0}) {name=outBaseColor, nodeColor=<rgba 0.1016 0 1 1>}
 ]}
